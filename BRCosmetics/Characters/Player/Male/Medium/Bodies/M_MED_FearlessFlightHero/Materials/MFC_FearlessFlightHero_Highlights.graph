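# Material function that adds stylized comic-book highlights to the FearlessFlightHero character's material attributes, behind static switches.
# Screen-projected rim vectors and sun-direction Lambert terms drive rim, halftone-dot and shadow-hatching masks.
# Procedural dots and hatching use actor-anchored, camera-facing virtual-plane UVs, with FOV-adjusted scale and stepped camera-distance values.
# FX texture channels mask suit regions for biplanar pattern, metallic, roughness and emissive changes, plus optional skin/hair shading.
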MaterialFunction /BRCosmetics/Characters/Player/Male/Medium/Bodies/M_MED_FearlessFlightHero/Materials/MFC_FearlessFlightHero_Highlights
graph {nodes [
  AtmosphericLightVector()
  FunctionOutput(None=StaticSwitchParameter(True=NamedRerouteUsage(), False=NamedRerouteUsage()) {group=01_FearlessFlightHero, parameterName=UseFearlessFlightHeroHighlights, sortPriority=0})
  n1 = GetMaterialAttributes(None=NamedRerouteUsage()) {attributeGetTypes [#1]}
  n2 = NamedRerouteUsage()
  n3 = ComponentMask(None=@n2) {b=true, g=true, r=true}
  n4 = ComponentMask(None=@n2) {a=true}
  n5 = Multiply(A=@n3, B=@n4)
  n6 = NamedRerouteUsage()
  n7 = ComponentMask(None=@n6) {b=true, g=true, r=true}
  n8 = ComponentMask(None=@n6) {a=true}
  n9 = Subtract(A=NamedRerouteUsage(), B=Multiply(A=@n7, B=@n8))
  n10 = Multiply(A=@n1.BaseColor, B=Saturate(None=Add(A=@n5, B=@n9)))
  n11 = Multiply(
   A=Multiply(A=@n8, B=NamedRerouteUsage())
   B=ScalarParameter() {group=01_FearlessFlightHero, parameterName=AdditiveShadowHatching, sliderMax=1}
  )
  n12 = Saturate(None=@n11)
  n13 = Multiply(
   A=Multiply(A=@n4, B=NamedRerouteUsage())
   B=ScalarParameter() {group=01_FearlessFlightHero, parameterName=AdditiveHighlightDots, sliderMax=1}
  )
  Multiply(A=Multiply(A=@n10, B=OneMinus(None=@n12)), B=OneMinus(None=Saturate(None=@n13)))
  n14 = TextureObjectParameter() {
   group=01_FearlessFlightHero
   parameterName=SuitTilingPattern
   samplerType=SAMPLERTYPE_Masks
   sortPriority=1
   texture=/BRCosmetics/Characters/Player/Male/Medium/Bodies/M_MED_FearlessFlightHero/Textures/T_M_MED_FearlessFlightHero_Pattern
  }
  n15 = ScalarParameter() {defaultValue=18, group=01_FearlessFlightHero, parameterName=SuitPatternScale, sliderMax=32}
  NamedRerouteDeclaration(None=ComponentMask(None=MaterialFunctionCall!(
     ColorTex=@n14
     NormalTex
     Pos
     Nrm
     ColorUVScale=@n15
     ColorBlendFactor
     NormalStrength
     NormalUVScale
     NormalBlendFactor
     Local Space Inputs
    ) {materialFunction=/Game/Global/MaterialFunctions/MF_BiPlanar}.Color) {r=true}) {name=SuitPattern, nodeColor=<rgba 1 1 1 1>}
  n16 = VectorParameter() {group=01_FearlessFlightHero, parameterName=RimColorDarkBlue, sortPriority=10, defaultValue=<rgba 0.0018 0.0065 0.0497 1>}
  n17 = NamedRerouteUsage()
  n18 = ScalarParameter() {group=01_FearlessFlightHero, parameterName=RimDarkBlueBias, sliderMax=1, sliderMin=-1}
  n19 = Multiply(A=AppendVector(A=@n16.RGB, B=@n16.RGB), B=Ceil(None=Add(A=@n17, B=@n18)))
  n20 = VectorParameter() {group=01_FearlessFlightHero, parameterName=RimColorBlue, sortPriority=10, defaultValue=<rgba 0.0015 0.0685 0.3372 1>}
  n21 = AppendVector(A=@n20.RGB, B=@n20.RGB)
  n22 = NamedRerouteUsage()
  n23 = ScalarParameter() {defaultValue=-0.2, group=01_FearlessFlightHero, parameterName=RimBlueBias, sliderMax=1, sliderMin=-1}
  n24 = NamedRerouteDeclaration(None=Ceil(None=Add(A=@n22, B=@n23))) {name=AccentRimMaskCell, nodeColor=<rgba 0.6658 0.7033 0.9131 1>}
  NamedRerouteDeclaration(None=Saturate(None=LinearInterpolate(A=@n19, B=@n21, Alpha=@n24))) {name=AccentRim, nodeColor=<rgba 0 0.0489 1 1>}
  n25 = VectorParameter() {group=01_FearlessFlightHero, parameterName=RimLighVector, sortPriority=10, defaultValue=<rgba -1.5 0 1 1>}
  n26 = NamedRerouteDeclaration(None=DotProduct(
    A=MaterialFunctionCall!(Vector=@n25.RGB) {materialFunction=/Game/Packages/Fortress_SharedMaterials/Anime_Master_Material/MF_ProjectToScreenSpace}.Result
    B=PixelNormalWS()
   )) {name=RimGradient, nodeColor=<rgba 0 0.3906 1 1>}
  n27 = NamedRerouteDeclaration(None=ScalarParameter() {defaultValue=-0.5, group=01_FearlessFlightHero, parameterName=YMask_Offset, sliderMax=1, sliderMin=-1}) {name=YmaskOffset, nodeColor=<rgba 0 1 0.2434 1>}
  n28 = DotProduct(A=VertexInterpolator(VS=PreSkinnedNormal()).PS, B=Constant3Vector() {constant=<rgba 0 1 0 0>})
  n29 = Multiply(A=@n28, B=@n28)
  n30 = Multiply(A=@n29, B=@n29) {desc=reduce rim light on the Y axis}
  NamedRerouteDeclaration(None=LinearInterpolate(A=@n26, B=Add(A=@n26, B=@n27) {constB=-0.25}, Alpha=NamedRerouteDeclaration(None=@n30) {name=YMask, nodeColor=<rgba 0 1 0.5078 1>}) {constA=1, constB=16}) {name=AccentRimMask, nodeColor=<rgba 0.3248 0.414 0.9131 1>}
  n31 = Constant()
  n32 = Constant() {r=0.1}
  n33 = VectorParameter() {group=01_FearlessFlightHero, parameterName=RimDotsVector, sortPriority=10, defaultValue=<rgba -0.85 0.3 0.5 1>}
  n34 = MaterialFunctionCall(Vector=@n33.RGB) {materialFunction=&MF_ProjectToScreenSpace}
  n35 = PixelNormalWS()
  n36 = ScalarParameter() {defaultValue=-0.1, group=01_FearlessFlightHero, parameterName=RimDotsOffset, sliderMax=0.5, sliderMin=-0.5, sortPriority=20}
  n37 = Multiply(
   A=Add(A=DotProduct(A=@n34.Result, B=@n35), B=@n36) {constB=-0.25}
   B=ScalarParameter() {defaultValue=1, group=01_FearlessFlightHero, parameterName=RimDotsTransition, sliderMax=2, sortPriority=20}
  ) {constB=1.5}
  n38 = Add(A=@n37, B=NamedRerouteUsage()) {constB=-0.25}
  n39 = NamedRerouteUsage()
  n40 = Saturate(None=Subtract(A=LinearInterpolate(A=@n37, B=@n38, Alpha=@n39) {constA=1, constB=16}, B=NamedRerouteUsage()))
  n41 = Add(
   A=NamedRerouteUsage()
   B=ScalarParameter() {defaultValue=-0.6, group=01_FearlessFlightHero, parameterName=RimWhiteBias, sliderMax=1, sliderMin=-1}
  )
  n42 = NamedRerouteDeclaration(None=Max(A=SmoothStep(Min=@n31, Max=@n32, Value=@n40), B=Ceil(None=@n41))) {name=RimDotsMask, nodeColor=<rgba 0.4766 1 0 1>}
  n43 = VectorParameter() {group=01_FearlessFlightHero, parameterName=RimDotsColor, sortPriority=10, defaultValue=<rgba 0.9 0.9 0.9 1>}
  NamedRerouteDeclaration(None=AppendVector(A=Multiply(A=@n42, B=@n43.RGB), B=@n43.A)) {name=RimDots, nodeColor=<rgba 1 0 0.6797 1>}
  n44 = Multiply(A=NamedRerouteUsage()) {constB=-1}
  n45 = ScalarParameter() {group=01_FearlessFlightHero, parameterName=ShadowHatchingOffset, sliderMax=0.5, sliderMin=-0.5, sortPriority=20}
  n46 = Subtract(A=Multiply(A=Add(A=@n44, B=@n45) {constB=-0.25}) {constB=0.5, desc=:/}, B=NamedRerouteUsage())
  n47 = NamedRerouteDeclaration(None=SmoothStep(Min=Constant(), Max=Constant() {r=0.1}, Value=Saturate(None=@n46))) {name=ShadowHatchingMask, nodeColor=<rgba 0.3594 0.3594 0.3594 1>}
  n48 = VectorParameter() {group=01_FearlessFlightHero, parameterName=ShadowHatchingColor, sortPriority=10, defaultValue=<rgba 0.5 0.5 0.5 0.5>}
  NamedRerouteDeclaration(None=AppendVector(A=Multiply(A=@n47, B=@n48.RGB), B=@n48.A)) {name=ShadowHatching, nodeColor=<rgba 0.099 0.099 0.099 1>}
  n49 = TextureSample() {
   samplerType=SAMPLERTYPE_Masks
   texture=/BRCosmetics/Characters/Player/Male/Medium/Bodies/M_MED_FearlessFlightHero/Textures/T_M_MED_FearlessFlightHero_Body_FX
  }
  n50 = FunctionInput(Preview=@n49.RGBA) {inputName=FX, inputType=FunctionInput_Vector4, sortPriority=1}
  NamedRerouteDeclaration(None=ChannelMaskParameter(None=@n50) {group=01_FearlessFlightHero, parameterName=SuitBlackMaskChannel, sortPriority=3}.R) {name=SuitBlackMask, nodeColor=<rgba 0.0156 0.0156 0.0156 1>}
  n51 = MaterialFunctionCall() {materialFunction=&CameraDirectionVector}
  n52 = Multiply(A=Constant3Vector() {constant=<rgba 1 1 0 0>}, B=@n51.Result)
  n53 = Add(A=Multiply(B=Normalize(VectorInput=@n52)) {constA=0.25}, B=@n51.Result)
  n54 = Multiply(A=@n51.Result, B=ScalarParameter() {defaultValue=10, group=01_FearlessFlightHero, parameterName=ScreenPostionDepthBias, sliderMax=100}) {constB=50}
  n55 = NamedRerouteDeclaration(None=Add(
    A=ActorPositionWS()
    B=VectorParameter() {group=01_FearlessFlightHero, parameterName=VirtualPlanePivotOffset, defaultValue=<rgba 0 0 90 1>}.RGB
   )) {name=AdjustedActorPosition, nodeColor=<rgba 0.8281 0 1 1>}
  n56 = Constant() {r=64}
  n57 = Constant() {r=128}
  n58 = InverseLinearInterpolate(
   A=Constant() {r=45}
   B=Constant() {r=90}
   Value=MaterialFunctionCall() {materialFunction=/Engine/Functions/Engine_MaterialFunctions02/Utility/FOV}.FOV angle in degrees
  )
  NamedRerouteDeclaration(None=MaterialFunctionCall(Plane Normal Axis=Normalize(VectorInput=@n53), Plane Center=Add(A=@n54, B=@n55), UV Scale=LinearInterpolate(A=@n56, B=@n57, Alpha=@n58)) {materialFunction=&VirtualPlaneCoordinates}.UVs) {name=ObjectAlighedScreenPosition, nodeColor=<rgba 1 0.867 0 1>}
  n59 = ChannelMaskParameter(None=@n50) {group=01_FearlessFlightHero, maskChannel=Green, parameterName=SkinHairShadingMaskChannel, sortPriority=3, defaultValue=<rgba 0 1 0 0>}
  NamedRerouteDeclaration(None=@n59.R) {name=SkinHairShadingMask, nodeColor=<rgba 0.1519 1 0 1>}
  n60 = Custom() {code="return View.DirectionalLightDirection;", description=DirectionalLightDirection, inputs []}
  n61 = CollectionParameter() {
   collection=&FortniteMaterialParameters
   parameterName=SunlightDirectionalVector
  }
  n62 = ComponentMask(None=@n61) {b=true, g=true, r=true}
  n63 = If(A=DotProduct(A=@n60, B=@n60), "A > B"=@n60, "A == B"=@n62, "A < B"=@n62)
  n64 = PixelNormalWS()
  n65 = Multiply(A=NamedRerouteDeclaration(None=DotProduct(A=@n63, B=@n64)) {name=LdotN, nodeColor=<rgba 0.2656 0 1 1>}) {constB=0.5}
  NamedRerouteDeclaration(None=Saturate(None=Add(A=@n65) {constB=0.5})) {name=Lambert, nodeColor=<rgba 0 0.9297 1 1>}
  n66 = Multiply(
   A=NamedRerouteUsage()
   B=Divide(
    A=ScalarParameter() {defaultValue=300, group=01_FearlessFlightHero, parameterName=HatchingAmount, sliderMax=512, sortPriority=20}
    B=NamedRerouteUsage()
   )
  )
  n67 = Constant() {r=1}
  n68 = Frac(None=ComponentMask(None=Rotator(Coordinate=@n66, Time=@n67) {speed=0.75}) {g=true})
  NamedRerouteDeclaration(None=Abs(None=Add(A=@n68) {constB=-0.5})) {name=HatchingGradient, nodeColor=<rgba 0.0041 0.1261 1 1>}
  n69 = GetMaterialAttributes(None=NamedRerouteUsage()) {attributeGetTypes [
    #2
    #1
    #3
    #4
   ]}
  n70 = Multiply(
   A=GetMaterialAttributes(None=NamedRerouteUsage()) {attributeGetTypes [#1]}.BaseColor
   B=NamedRerouteUsage()
  )
  n71 = ScalarParameter() {defaultValue=0.6, group=01_FearlessFlightHero, parameterName=SuitLogoEmissiveStrength, sliderMax=1}
  n72 = LinearInterpolate(A=NamedRerouteUsage(), B=NamedRerouteUsage(), Alpha=NamedRerouteUsage())
  n73 = Multiply(A=ComponentMask(None=@n72) {b=true, g=true, r=true}, B=Multiply(A=ComponentMask(None=@n72) {a=true}, B=NamedRerouteUsage())) {desc=Left RIm}
  n74 = Saturate(None=Add(A=Multiply(A=@n70, B=@n71) {constB=0.5}, B=@n73))
  n75 = NamedRerouteUsage()
  n76 = NamedRerouteUsage()
  n77 = Multiply(A=NamedRerouteUsage(), B=Saturate(None=Subtract(A=@n75, B=@n76)))
  n78 = Multiply(A=Max(A=NamedRerouteUsage(), B=NamedRerouteUsage()), B=NamedRerouteUsage())
  n79 = SetMaterialAttributes!(
   MaterialAttributes=@n69.MaterialAttributes
   EmissiveColor=Add(
    A=@n69.EmissiveColor
    B=MaterialFunctionCall!(In=@n74, LUT, LUT Multiplier) {materialFunction=/Game/Packages/Fortress_SharedMaterials/Anime_Master_Material/MF_ApplyLUT}.Result
   )
   BaseColor=LinearInterpolate(A=Add(A=@n69.BaseColor, B=@n77), B=@n73, Alpha=Saturate(None=@n78))
   Metallic=LinearInterpolate(
    A=@n69.Metallic
    B=ScalarParameter() {defaultValue=0.6, group=01_FearlessFlightHero, parameterName=SuitBlackMetallic, sliderMax=1}
    Alpha=NamedRerouteUsage()
   )
   Roughness=LinearInterpolate(
    A=@n69.Roughness
    B=ScalarParameter() {defaultValue=0.6, group=01_FearlessFlightHero, parameterName=SuitLogoRoughness, sliderMax=1}
    Alpha=NamedRerouteUsage()
   )
  ) {
   desc=screen projected rim and hatching
   attributeSetTypes [
    #2
    #1
    #3
    #4
   ]
  }
  NamedRerouteDeclaration(None=@n79) {name=AccentRimMaterial, nodeColor=<rgba 0 0.12 1 1>}
  n80 = Power(
   Base=DotProduct(A=PixelNormalWS(), B=CameraVectorWS())
   Exp=ScalarParameter() {defaultValue=12, group=01_FearlessFlightHero, parameterName=PatternShadingExponent, sliderMax=64, sortPriority=20}
  ) {constExponent=64}
  n81 = Add(
   A=Multiply(A=NamedRerouteUsage(), B=@n80)
   B=Multiply(A=@n80, B=ScalarParameter() {defaultValue=0.04, group=01_FearlessFlightHero, parameterName=PatternBaseReflectFraction, sliderMax=1})
  )
  n82 = VectorParameter() {group=01_FearlessFlightHero, parameterName=SuitPatternColor, sortPriority=10, defaultValue=<rgba 0.0095 0.0943 0.2604 1.5>}
  NamedRerouteDeclaration(None=Multiply(A=Saturate(None=@n81), B=Multiply(A=@n82.RGB, B=@n82.RGB))) {name=SuitPatternTinted, nodeColor=<rgba 0 0.3233 1 1>}
  n83 = Multiply(
   A=NamedRerouteUsage()
   B=Divide(
    A=ScalarParameter() {defaultValue=200, group=01_FearlessFlightHero, parameterName=DotsAmount, sliderMax=512, sortPriority=20}
    B=NamedRerouteUsage()
   )
  )
  n84 = Constant() {r=1}
  n85 = Add(A=Frac(None=Rotator(Coordinate=@n83, Time=@n84) {speed=0.75})) {constB=-0.5}
  n86 = Constant()
  NamedRerouteDeclaration(None=Add(A=Distance(A=@n85, B=@n86)) {constB=-0.1, desc=to get full transition}) {name=DotsGradient, nodeColor=<rgba 1 0.0938 0 1>}
  n87 = NamedRerouteUsage()
  n88 = ScalarParameter() {defaultValue=-0.3, group=01_FearlessFlightHero, parameterName=HighlightDotsOffset, sliderMax=0.5, sliderMin=-0.5, sortPriority=20}
  n89 = Subtract(
   A=Multiply(
    A=Add(A=@n87, B=@n88) {constB=-0.25}
    B=ScalarParameter() {defaultValue=1.5, group=01_FearlessFlightHero, parameterName=HighlightDotsTransition, sliderMax=2, sortPriority=20}
   ) {constB=1.5}
   B=NamedRerouteUsage()
  )
  n90 = NamedRerouteDeclaration(None=SmoothStep(Min=Constant(), Max=Constant() {r=0.1}, Value=Saturate(None=@n89))) {name=HighlightDotsMask, nodeColor=<rgba 0.4766 1 0 1>}
  n91 = VectorParameter() {group=01_FearlessFlightHero, parameterName=HighlightDotsColor, sortPriority=10, defaultValue=<rgba 0.9 0.9 0.9 1>}
  NamedRerouteDeclaration(None=AppendVector(A=Multiply(A=@n90, B=@n91.RGB), B=@n91.A)) {name=HighlightDots, nodeColor=<rgba 1 0 0.6797 1>}
  n92 = ChannelMaskParameter(None=@n50) {group=01_FearlessFlightHero, maskChannel=Blue, parameterName=SuitLogoMaskChannel, sortPriority=3, defaultValue=<rgba 0 0 1 0>}
  NamedRerouteDeclaration(None=@n92.R) {name=SuitLogoMask, nodeColor=<rgba 1 0.0509 0 1>}
  n93 = ChannelMaskParameter(None=@n50) {group=01_FearlessFlightHero, maskChannel=Green, parameterName=SuitRedMaskChannel, sortPriority=3, defaultValue=<rgba 0 1 0 0>}
  NamedRerouteDeclaration(None=@n93.R) {name=SuitRedMask, nodeColor=<rgba 1 0.0143 0 1>}
  n94 = ChannelMaskParameter(None=@n50) {group=01_FearlessFlightHero, maskChannel=Alpha, parameterName=SuitRubberMaskChannel, sortPriority=3, defaultValue=<rgba 0 0 0 1>}
  NamedRerouteDeclaration(None=@n94.R) {name=SuitRubberMask, nodeColor=<rgba 0 0.0255 0.2969 1>}
  n95 = TextureSample() {
   samplerType=SAMPLERTYPE_Masks
   texture=/BRCosmetics/Characters/Player/Male/Medium/Bodies/M_MED_FearlessFlightHero/Textures/T_M_MED_FearlessFlightHero_Body_S
  }
  n96 = TextureSample() {
   samplerType=SAMPLERTYPE_Normal
   texture=/BRCosmetics/Characters/Player/Male/Medium/Bodies/M_MED_FearlessFlightHero/Textures/T_M_MED_FearlessFlightHero_Body_N
  }
  n97 = SetMaterialAttributes!(
   MaterialAttributes
   BaseColor=TextureSample() {texture=/BRCosmetics/Characters/Player/Male/Medium/Bodies/M_MED_FearlessFlightHero/Textures/T_M_MED_FearlessFlightHero_Body_D}.RGB
   Specular=@n95.R
   Metallic=@n95.R
   Roughness=@n95.R
   Normal=@n96.RGB
  ) {attributeSetTypes [
    #1
    #5
    #3
    #4
    #6
   ]}
  NamedRerouteDeclaration(None=FunctionInput(Preview=@n97) {inputName=MA, inputType=FunctionInput_MaterialAttributes}) {name=MA, nodeColor=<rgba 0.3359 1 0 1>}
  n98 = LinearInterpolate(A=LinearInterpolate(A=@n10, B=@n7, Alpha=@n11), B=@n3, Alpha=@n13)
  n99 = VectorParameter() {group=01_FearlessFlightHero, parameterName=SkinHighlightColor, defaultValue=<rgba 0.9 0.9 0.9 1>}
  n100 = Multiply(A=@n99.RGB, B=@n99.RGB)
  n101 = Saturate(None=Add(A=NamedRerouteUsage(), B=ScalarParameter() {group=01_FearlessFlightHero, parameterName=SkinHighlightBias, sliderMax=1, sliderMin=-1}))
  n102 = SetMaterialAttributes!(MaterialAttributes=@n1.MaterialAttributes, BaseColor=@n98, EmissiveColor=Add(A=Multiply(A=@n98) {constB=0.5}, B=Multiply(A=@n100, B=@n101))) {attributeSetTypes [#1, #2]}
  n103 = BlendMaterialAttributes(A=NamedRerouteUsage(), B=@n102, Alpha=NamedRerouteUsage())
  NamedRerouteDeclaration(None=StaticSwitchParameter(
    True=StaticSwitchParameter(True=@n103, False=@n102) {group=01_FearlessFlightHero, parameterName=UseSkinHairShading_FXMask, sortPriority=1}
    False=NamedRerouteUsage()
   ) {group=01_FearlessFlightHero, parameterName=UseSkinHairShading, sortPriority=1}) {name=MA_Out, nodeColor=<rgba 1 0.0703 0 1>}
  n104 = Subtract(A=Distance(A=NamedRerouteUsage(), B=CameraPositionWS())) {constB=100}
  n105 = ScalarParameter() {defaultValue=6, group=01_FearlessFlightHero, parameterName=DistanceBlendingSteps, sliderMax=8, sliderMin=1, sortPriority=20}
  n106 = Multiply(A=Saturate(None=Divide(A=@n104) {constB=1200}), B=@n105) {constB=4}
  NamedRerouteDeclaration(None=Max(A=Ceil(None=@n106))) {name=SteppedDistanceBlending, nodeColor=<rgba 0.4766 1 0 1>}
  SkyAtmosphereLightDirection()
 ]}
