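# Animation blueprint for the Fearless Flight Hero medium male player body cosmetic, driving its secondary deformation on top of the parent character's pose.
# The AnimGraph copies the pose from the attached parent mesh, then runs it through a deform Control Rig followed by a twist Control Rig.
# Each update, the EventGraph computes a GravityOverride vector through a deprecated gravity override state function using game time and override parameters.
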AnimBlueprint /BRCosmetics/Characters/Player/Male/Medium/Bodies/M_MED_FearlessFlightHero/Meshes/M_MED_FearlessFlightHero_AnimBP
target_skeleton=&/M_MED_FearlessFlightHero_Skeleton
graphs [
 AnimGraph {nodes [
   n1 = Control Rig(Source=Copy Pose From Mesh() {bUseAttachedParent=true}) {
    alphaBoolBlend {blendOption=Linear}
    controlRigAssetReference {blueprintRigClass=&/M_MED_DeformRig_fearlessFlightHero.M_MED_DeformRig_fearlessFlightHero_C}
   }
   n2 = Control Rig(Source=@n1) {
    alphaBoolBlend {blendOption=Linear}
    controlRigAssetReference {blueprintRigClass=&/M_MED_FearlessFlightHero_ControlRig_Twist.M_MED_FearlessFlightHero_ControlRig_Twist_C}
   }
   Output Pose(Result=@n2) {layerGroup=None}
  ]}
 EventGraph {nodes [
   Event BlueprintUpdateAnimation()
   > n1 = Gravity Override State Function DEPRECATED(CharacterPartAnimInstance=Self-Reference(), GameTimeInSeconds=GetGameTimeInSeconds(), GravityOverrideParams=Get Gravity Override Params())
   > Set GravityOverride(GravityOverride=@n1)
  ]}
]
variables [
 forwards_backwards_alpha {type=number}
 Gravity Override Params {type=GravityOverrideParamsStruct}
 GravityOverride {type=Vector}
 Is Hang Glider {type=number}
 Is Pistol {type=boolean}
 Is Rifle {type=boolean}
 Is Sky Diving {type=number}
 Is Skydive Floating Float {type=number}
 Is Wearing Backbling {type=boolean}
 left_right_alpha {type=number}
 Skeletal Mesh {type=SkeletalMeshComponent}
 thighOtAlpha {type=number}
 trapAlpha {type=number}
]
