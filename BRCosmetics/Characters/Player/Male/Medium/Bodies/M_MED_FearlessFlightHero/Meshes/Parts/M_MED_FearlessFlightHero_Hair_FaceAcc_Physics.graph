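# Physics asset for the hair and face accessory part of the FearlessFlightHero medium male player cosmetic.
# It defines capsule bodies on two bones: a small one on hair_MSTR and a larger one on hair_c, offset from its bone and rotated 90 degrees.
# A single constraint joins hair_c to hair_MSTR with near-zero swing and twist limit angles.
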PhysicsAsset /BRCosmetics/Characters/Player/Male/Medium/Bodies/M_MED_FearlessFlightHero/Meshes/Parts/M_MED_FearlessFlightHero_Hair_FaceAcc_Physics
bodies {
 hair_MSTR [{shapeType=Capsule, length=2, radius=1, shapeName=hair_MSTR_capsule}]
 hair_c [{shapeType=Capsule, length=7, radius=5, shapeName=hair_c_capsule, center=<5.1355 0 0>, rotation=<rot 90 0 0>}]
}
constraints [{bone1Name=hair_c, bone2Name=hair_MSTR, swing1LimitDegrees=1, swing2LimitDegrees=0, twistLimitDegrees=0}]
